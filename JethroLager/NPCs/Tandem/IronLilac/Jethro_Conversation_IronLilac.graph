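# Conversation database for the Iron Lilac NPC in JethroLager, defining its greeting and top-level conversation branches.
# The entry point leads to a speech node whose default greeting gives way to prioritized contextual lines for a first-ever meeting or for specific player outfits, matched by cosmetic character IDs.
# The speech node then branches through link nodes to the shared NPC quest library, Jethro services, and Iron Lilac's own dialogue entry.
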ConversationDatabase /JethroLager/NPCs/Tandem/IronLilac/Jethro_Conversation_IronLilac
entry_points [{tag=Jethro.Conversation.Entry.IronLilac, destinations [#1]}]
nodes [
 ConversationEntryPointNode_7 {
  id=#1
  class=ConversationEntryPointNode
  outputs [#2]
  properties {entryTag {tagName=Jethro.Conversation.Entry.IronLilac}}
 }
 Speech_C_4 {
  id=#2
  class=Speech_C
  outputs [#3, #4, #5]
  properties {
   nodeName=""
   generalConfig {
    defaultMessage=How're you holding up?
    contextualMessages [
     {
      message=I'm glad you're here. This is going to get... interesting.
      contextRequirements [{requirement {class=FortControllerRequirement_IsFirstEverConversationWithNPC}}]
      priority {curve {rowName=Tandem.Speech.Priority.FirstEverInteraction}}
     }
     {
      message=...
      contextRequirements [{requirement {
         class=FortControllerRequirement_HasCID
         properties {allowedCIds [
           &CID_A_422_Athena_Commando_M_Realm
           /BRCosmetics/Athena/Items/Cosmetics/Characters/Character_SequinPie
           /BRCosmetics/Athena/Items/Cosmetics/Characters/Character_MuffinLadle_Gas
          ]}
        }}]
      priority {curve {rowName=Tandem.Speech.Priority.OutfitReaction}}
     }
     {
      message="Keep your eyes... peeled. A few of our new friends seem a little curious about you. "
      contextRequirements [{requirement {
         class=FortControllerRequirement_HasCID
         properties {allowedCIds [&CID_349_Athena_Commando_M_Banana]}
        }}]
      priority {curve {rowName=Tandem.Speech.Priority.OutfitReaction}}
     }
     {
      message=The threat is not gone... but it may have to wait.
      contextRequirements [{requirement {class=FortControllerRequirement_HasCID, properties {allowedCIds [&Character_CubicVice_Dual]}}}]
      priority {curve {rowName=Tandem.Speech.Priority.OutfitReaction}}
     }
     {
      message=You're not him. I can tell.
      contextRequirements [{requirement {class=FortControllerRequirement_HasCID, properties {allowedCIds [&Character_MistMylar_Claw]}}}]
      priority {curve {rowName=Tandem.Speech.Priority.OutfitReaction}}
     }
     {
      message=I don't know if our new friends need us... but I can't just sit back and watch.
      contextRequirements [{requirement {
         class=FortControllerRequirement_HasCID
         properties {allowedCIds [&Character_UndergroundRebel_Fashion]}
        }}]
      priority {curve {rowName=Tandem.Speech.Priority.OutfitReaction}}
     }
    ]
   }
  }
 }
 ConversationLinkNode_2 {
  id=#5
  class=ConversationLinkNode
  properties {remoteEntryTag {tagName=Jethro.Conversation.Entry.IronLilac.Dialogue}}
 }
 ConversationLinkNode_0 {id=#4, class=ConversationLinkNode, properties {remoteEntryTag {tagName=Conversation.Entry.JethroServices}}}
 ConversationLinkNode_1 {id=#3, class=ConversationLinkNode, properties {remoteEntryTag {tagName=Conversation.Entry.NPCLibrary.Quests}}}
]
