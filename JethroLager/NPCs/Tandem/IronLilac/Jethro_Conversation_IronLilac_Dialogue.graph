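# Conversation database for the Iron Lilac NPC's "Ask about..." dialogue in JethroLager, entered through the Jethro.Conversation.Entry.IronLilac.Dialogue tag.
# A "What's up?" speech node offers two topics, Daigo and The Empire, plus a Back option.
# Each topic plays two chained speech lines, then both converge on a shared ReturnToStart node labelled BACK.
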ConversationDatabase /JethroLager/NPCs/Tandem/IronLilac/Jethro_Conversation_IronLilac_Dialogue
entry_points [{tag=Jethro.Conversation.Entry.IronLilac.Dialogue, destinations [#1]}]
nodes [
 ConversationEntryPointNode_0 {
  id=#1
  class=ConversationEntryPointNode
  outputs [#2]
  properties {entryTag {tagName=Jethro.Conversation.Entry.IronLilac.Dialogue}}
 }
 Speech_C_0 {
  id=#2
  class=Speech_C
  outputs [#3, #4, #5]
  properties {nodeName="", softTaskIcon=&T_UI_NPC_question, generalConfig {defaultMessage=What's up?}}
  sub_nodes [ConversationChoiceNode_5 {class=ConversationChoiceNode, properties {defaultChoiceDisplayText=Ask about...}}]
 }
 Speech_C_1 {
  id=#5
  class=Speech_C
  outputs [#6]
  properties {
   nodeName=""
   softTaskIcon=&T_UI_NPC_question
   generalConfig {defaultMessage="If I had to guess, I'd say they're looking for power..."}
  }
  sub_nodes [ConversationChoiceNode_1 {class=ConversationChoiceNode, properties {defaultChoiceDisplayText=The Empire}}]
 }
 Speech_C_2 {
  id=#6
  class=Speech_C
  outputs [#7]
  properties {nodeName="", softTaskIcon=&T_UI_NPC_question, generalConfig {defaultMessage=And you know how that ends.}}
 }
 ReturnToStart_C_23 {
  id=#7
  class=ReturnToStart_C
  properties {softTaskIcon=&T_UI_NPC_Back}
  sub_nodes [ConversationChoiceNode_32 {class=ConversationChoiceNode, properties {defaultChoiceDisplayText=BACK}}]
 }
 Back_C_0 {
  id=#4
  class=Back_C
  sub_nodes [ConversationChoiceNode_7 {class=ConversationChoiceNode, properties {defaultChoiceDisplayText=BACK}}]
 }
 Speech_C_7 {
  id=#3
  class=Speech_C
  outputs [#8]
  properties {
   nodeName=""
   softTaskIcon=&T_UI_NPC_question
   generalConfig {defaultMessage=He's brought us all to a dangerous place. I don't know what happens next...}
  }
  sub_nodes [ConversationChoiceNode_0 {class=ConversationChoiceNode, properties {defaultChoiceDisplayText=Daigo}}]
 }
 Speech_C_15 {
  id=#8
  class=Speech_C
  outputs [#7]
  properties {nodeName="", softTaskIcon=&T_UI_NPC_question, generalConfig {defaultMessage=But it's far from over.}}
 }
]
